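# Editor utility blueprint whose Generate LOD Report function runs an embedded Python script to report LOD types of Static Meshes selected in the Content Browser.
# Meshes with more than one LOD are sorted by LOD 1's reduction triangle percentage: below 1.0 counts as autogenerated, otherwise as custom.
# It writes the mesh names to LOD_Autogen.csv and LOD_Custom.csv on the user's Desktop; single-LOD meshes appear in neither.
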EditorUtilityBlueprint /Game/ArtTools/Environment/LODReport/LODReport
graphs [
 EventGraph {links [], nodes []}
 Generate LOD Report {nodes [
   Generate LOD Report() {is=title K2Node_FunctionEntry}
   > Execute Python Script(PythonScript="import unreal\r\nimport csv\r\nimport os\r\n\r\ndef run_mesh_lod_report():\r\n    unreal.log(\"--- Starting Static Mesh LOD Report ---\")\r\n    desktop_path = os.path.join(os.path.expanduser(\"~\"), \"Desktop\")\r\n    \r\n    ## 1. Get Selection from Content Browser\r\n    selected_assets = unreal.EditorUtilityLibrary.get_selected_assets()\r\n    static_meshes = [a for a in selected_assets if isinstance(a, unreal.StaticMesh)]\r\n    \r\n    if not static_meshes:\r\n        unreal.log_warning(\"No Static Meshes selected in Content Browser.\")\r\n        return\r\n\r\n    autogen_data = [[\"Mesh Name\"]]\r\n    custom_data = [[\"Mesh Name\"]]\r\n\r\n    ## 2. Analyze LODs\r\n    for mesh in static_meshes:\r\n        mesh_name = mesh.get_name()\r\n        lod_count = unreal.EditorStaticMeshLibrary.get_lod_count(mesh)\r\n        \r\n        if lod_count > 1:\r\n            ## Check LOD 1 settings\r\n            reduction = unreal.EditorStaticMeshLibrary.get_lod_reduction_settings(mesh, 1)\r\n            \r\n            ## If percent_triangles is 1.0, it is treated as a Custom/Imported LOD\r\n            if reduction.percent_triangles < 1.0:\r\n                autogen_data.append([mesh_name])\r\n            else:\r\n                custom_data.append([mesh_name])\r\n\r\n    ## 3. Export CSVs\r\n    try:\r\n        with open(os.path.join(desktop_path, \"LOD_Autogen.csv\"), 'w', newline='') as f:\r\n            csv.writer(f).writerows(autogen_data)\r\n        with open(os.path.join(desktop_path, \"LOD_Custom.csv\"), 'w', newline='') as f:\r\n            csv.writer(f).writerows(custom_data)\r\n        \r\n        unreal.log(f\"Elimination of manual checking complete. Reports saved to Desktop.\")\r\n    except Exception as e:\r\n        unreal.log_error(f\"File Error: {str(e)}\")\r\n\r\nif __name__ == \"__main__\":\r\n    run_mesh_lod_report()")
  ]}
]
variables []
